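# Material function for tentacle portal projectile effects that outputs an animated, turbulent 3D vertex displacement.
# Actor-relative position drives two volume-texture lookups: panning curl noise and a scrolling, rotating Voronoi sampled in local space.
# Inverted, power-shaped Voronoi pushes along the normal (local-space via the "IS Niagara" switch), summed with centred curl noise.
# A "Displace Fade" switch optionally tapers the result toward both U edges.
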MaterialFunction /PistolCanary/Assets/FX/TentaclePortalProjectiles/MaterialReferences/MF_VoidMask_3DTurb
graph {nodes [
  n1 = NamedRerouteUsage()
  n2 = Subtract(A=ComponentMask(None=TextureCoordinate()) {r=true}) {constB=0.5}
  n3 = OneMinus(None=Abs(None=Multiply(A=@n2) {constB=2}))
  FunctionOutput(None=StaticSwitchParameter(True=Multiply(A=@n1, B=@n3), False=@n1) {parameterName=Displace Fade})
  n4 = Add(A=NamedRerouteUsage(), B=Multiply(A=VectorParameter() {parameterName=Curl Noise Pan Dir, defaultValue=<rgba 50 100 -30 0>}.RGB, B=Time()))
  n5 = ScalarParameter() {defaultValue=0.0011, parameterName=Noise Size}
  n6 = Subtract(A=TextureSample(UVs=Multiply(A=@n4, B=@n5)) {samplerType=SAMPLERTYPE_LinearColor, texture=/Game/Effects/FogVolume/VT_CurlNoise_LOW}.RGB) {constB=0.5}
  n7 = ScalarParameter() {defaultValue=1.5, parameterName=Noise Scale}
  n8 = VertexNormalWS()
  n9 = Multiply(
   A=VectorParameter() {parameterName=Volonoi Noise Scale, defaultValue=<rgba 0.01 0.01 0.01 0>}.RGB
   B=ScalarParameter() {defaultValue=4, parameterName=Main Noise Scale}
  )
  n10 = Multiply(A=Time(), B=ScalarParameter() {defaultValue=10, parameterName=Main Displacement Anim Speed}) {constB=100}
  n11 = Transform(None=NamedRerouteUsage()) {transformSourceType=TRANSFORMSOURCE_World, transformType=TRANSFORM_Local}
  n12 = ComponentMask(None=@n11) {g=true, r=true}
  n13 = Multiply(A=@n10, B=ScalarParameter() {defaultValue=0.4, parameterName=Main Noise Rotate Rate}) {constB=0.15}
  n14 = Add(A=@n10, B=AppendVector(A=Rotator(Coordinate=@n12, Time=@n13), B=ComponentMask(None=@n11) {b=true}))
  n15 = TextureObjectParameter() {
   group=Volume Textures
   parameterName=Displacement Volume Tex
   samplerType=SAMPLERTYPE_LinearColor
   texture=/Game/Effects/Explosion_Update/Textures/VT_Voronoi_Seamless
  }
  n16 = TextureSample(UVs=Multiply(A=Multiply(A=@n9, B=@n14)) {constB=0.5}, Tex=@n15) {samplerType=SAMPLERTYPE_LinearColor, texture=&VT_Voronoi_Seamless}
  n17 = ScalarParameter() {defaultValue=3, parameterName=Shape Unevenness Scale}
  n18 = Multiply(
   A=StaticSwitchParameter(True=Transform(None=@n8) {transformSourceType=TRANSFORMSOURCE_World, transformType=TRANSFORM_Local}, False=@n8) {parameterName=IS Niagara}
   B=OneMinus(None=Power(Base=@n16.R, Exp=@n17) {constExponent=1.5})
  )
  n19 = ScalarParameter() {defaultValue=8, parameterName=Displacement}
  NamedRerouteDeclaration(None=Add(A=Multiply(A=@n6, B=@n7), B=Multiply(A=@n18, B=@n19))) {name=3D Displacement, nodeColor=<rgba 0 1 0.9766 1>}
  NamedRerouteDeclaration(None=Subtract(A=WorldPosition() {worldPositionShaderOffset=WPT_ExcludeAllShaderOffsets}.XYZ, B=ActorPositionWS())) {name=CurrentPos, nodeColor=<rgba 1 0.6563 0 1>}
 ]}
