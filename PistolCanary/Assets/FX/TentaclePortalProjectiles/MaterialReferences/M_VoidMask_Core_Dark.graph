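# Translucent unlit, two-sided void-effect material for Niagara mesh particles, filed under tentacle portal projectiles.
# Emissive takes hue-shifted particle colour or a camera-vector triplanar galaxy with distorted edge bands, plus optional Fresnel glow.
# Opacity combines particle alpha, swirl-warped flame noise, UV masks and switchable Fresnel, depth and camera fades.
# Vertex offset adds 3D turbulence to either spline thickening or distance-based normal displacement, driven by dynamic parameters.
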Material /PistolCanary/Assets/FX/TentaclePortalProjectiles/MaterialReferences/M_VoidMask_Core_Dark
graph {
 properties {
  bAutomaticallySetUsageInEditor=false
  bUsedWithNiagaraMeshParticles=true
  blendMode=BLEND_Translucent
  shadingModel=MSM_Unlit
  twoSided=true
  usedShadingModels=Unlit
 }
 nodes [
  n1 = ParticleColor()
  n2 = MaterialFunctionCall(
   Hue Shift Percentage=ScalarParameter() {parameterName=EmissiveHueShift}
   Texture=StaticSwitchParameter(True=NamedRerouteUsage(), False=@n1.RGB) {parameterName=2D Space}
  ) {materialFunction=&HueShift}
  n3 = Fresnel() {baseReflectFraction=0, exponent=1}
  n4 = ScalarParameter() {defaultValue=3, group=FresnelGlow, parameterName=FresnelGlow_Power}
  n5 = LinearInterpolate(
   A=@n2.Result
   B=VectorParameter() {group=FresnelGlow, parameterName=FresnelGlow_Color, defaultValue=<rgba 0.4745 0.102 1 1>}.RGB
   Alpha=Saturate(None=Power(Base=@n3, Exp=@n4))
  )
  MaterialFunctionCall(Emissive=Multiply(
    A=ScalarParameter() {defaultValue=1, parameterName=Emissive Multiply}
    B=StaticSwitchParameter(True=@n5, False=@n2.Result) {group=FresnelGlow, parameterName=Use_Fresnel_Glow}
   )) {materialFunction=&MF_FX_EmissiveAdjustments}
  n6 = Distance(A=ObjectPositionWS(), B=WorldPosition().XYZ)
  n7 = ScalarParameter() {defaultValue=0.01, parameterName=Radius}
  Multiply(A=OneMinus(None=Multiply(A=@n6, B=@n7)), B=VectorParameter() {parameterName=Color, defaultValue=<rgba 0.252 0 1 1>}.RGB)
  n8 = DynamicParameter() {parameterIndex=3, defaultValue=<rgba 1 0 0 1>, paramNames [Thicken, Shape Shift, Rand, Noise UV Scale]}
  n9 = Constant2Vector() {g=0.5, r=1}
  n10 = MaterialFunctionCall!(UV=TextureCoordinate(), Twist Center, Twist Edges) {materialFunction=/Game/Accessories/FORT_Backpacks/Backpack_F_MED_Celestial_Glow/FX/Materials/MF_PolarUV}
  n11 = Constant()
  n12 = MaterialFunctionCall(Float2=Power(Base=@n10.Polar UVs) {constExponent=3}) {materialFunction=&BreakOutFloat2Components}
  n13 = Multiply(A=Power(Base=OneMinus(None=@n12.G)) {constExponent=4}) {constB=-0.12}
  n14 = StaticSwitchParameter(True=Add(A=Multiply(A=@n10.Polar UVs, B=@n9), B=AppendVector(A=@n11, B=@n13)) {constB=0}, False=TextureCoordinate()) {parameterName=Use_RadialUvs_Flames}
  n15 = VectorParameter() {parameterName=OutlineUV, channelNames {a=Swirl V, b=Swirl U, g=V, r=U}, defaultValue=<rgba 1 1 1 1>}
  n16 = Time()
  n17 = ScalarParameter() {defaultValue=0.25, parameterName=Swirl Pan Speed}
  n18 = Add(
   A=Multiply(A=Multiply(A=@n8.Noise UV Scale, B=@n14), B=AppendVector(A=@n15.U, B=@n15.U))
   B=Multiply(A=Multiply(A=@n16, B=@n17), B=Constant2Vector() {g=1, r=0.5})
  )
  n19 = Multiply(A=@n14, B=AppendVector(A=@n15.Swirl U, B=@n15.Swirl U))
  n20 = Subtract(A=ComponentMask(None=TextureSample(UVs=@n19) {samplerType=SAMPLERTYPE_LinearColor, texture=/Game/VisualThreatManager/StormVisuals/LargeClouds/T_Swirl_Flowmap}.RGB) {g=true, r=true}) {constB=0.5}
  n21 = ScalarParameter() {defaultValue=0.5, parameterName=Swirl Strength}
  n22 = StaticSwitchParameter(True=Add(A=@n18, B=Multiply(A=@n20, B=@n21)), False=@n18) {parameterName=Use UV Swirl}
  n23 = TextureSampleParameter2D!(UVs=@n22, "DDX(UVs)"=@n10.DDX, "DDY(UVs)"=@n10.DDX, Apply View MipBias) {
   group=Warp
   mipValueMode=TMVM_Derivative
   parameterName=TexFlames
   samplerType=SAMPLERTYPE_LinearColor
   texture=/Game/Effects/Fort_Effects/Creatures/Elemental/Fire/MergedFireFlowMap
  }
  n24 = Subtract(A=StaticSwitchParameter(
    True=@n23.R
    False=TextureSample(UVs=@n22) {samplerType=SAMPLERTYPE_LinearColor, texture=&MergedFireFlowMap}.R
   ) {parameterName=Use_RadialUvs_Flames}) {constB=0.5}
  n25 = Subtract(A=TextureCoordinate()) {constB=0.5}
  n26 = OneMinus(None=Multiply(A=Abs(None=@n25)) {constB=2})
  n27 = ComponentMask(None=@n26) {g=true}
  NamedRerouteDeclaration(None=Power(Base=Add(A=@n24, B=@n27), Exp=ScalarParameter() {defaultValue=4, parameterName=SwirlPow})) {name=V_Gradient, nodeColor=<rgba 0.1016 0 1 1>}
  n28 = ComponentMask(None=@n26) {r=true}
  NamedRerouteDeclaration(None=Power(Base=@n28, Exp=ScalarParameter() {defaultValue=0.3, parameterName=Thicken Curve}) {constExponent=0.3}) {name=U_Mask, nodeColor=<rgba 0.8984 0 1 1>}
  n29 = MaterialFunctionCall(Float2=Multiply(
    A=TextureCoordinate()
    B=ScalarParameter() {defaultValue=1, group="2. Edge Distortion Band (Fresnel Distortion = 0)", parameterName=EdgeDistortion_Tiling}
   )) {materialFunction=&BreakOutFloat2Components}
  n30 = Add(
   A=@n29.G
   B=ScalarParameter() {defaultValue=0.3, group="2. Edge Distortion Band (Fresnel Distortion = 0)", parameterName=EdgeDistort_OffsetBottom}
  )
  n31 = ScalarParameter() {desc=Switches between using UV channel 0 or 1 for mapping distortion, group=1. Base, parameterName=Use_UV1_ForDistortion, sortPriority=-1}
  n32 = MaterialFunctionCall(Float2=LinearInterpolate(A=TextureCoordinate(), B=TextureCoordinate() {coordinateIndex=1}, Alpha=@n31)) {materialFunction=&BreakOutFloat2Components}
  n33 = AppendVector(
   A=Multiply(A=@n32.R, B=ScalarParameter() {defaultValue=1, group=Distortion_Macro, parameterName=Distortion_Macro_Tile_X, sortPriority=3})
   B=Multiply(A=@n32.G, B=ScalarParameter() {defaultValue=0.5, group=Distortion_Macro, parameterName=Distortion_Macro_Tile_Y, sortPriority=4})
  )
  n34 = AppendVector(
   A=ScalarParameter() {defaultValue=0.1, group=Distortion_Macro, parameterName=Distortion_Macro_Panspeed_X, sortPriority=1}
   B=ScalarParameter() {defaultValue=0.25, group=Distortion_Macro, parameterName=Distortion_Macro_Panspeed_Y, sortPriority=2}
  )
  n35 = MaterialFunctionCall!(Position, UVRadialScale) {materialFunction=/Game/Weapons/WeaponSkins/MaterialFunction/ObjectStabilizedScreenSpaceUVs}
  n36 = Multiply(
   A=MaterialFunctionCall(Vector or UVs=Subtract(A=@n35.StabilizedUVs) {constB=0.5}) {materialFunction=&VectorToRadialValue}.Radial Coordinates
   B=AppendVector(
    A=ScalarParameter() {group=Distortion_Macro, parameterName=Radial_Tiling_X}
    B=ScalarParameter() {group=Distortion_Macro, parameterName=Radial_Tiling_Y}
   )
  )
  n37 = AppendVector(
   A=ScalarParameter() {group=Distortion_Macro, parameterName=Radial_Panning_X}
   B=ScalarParameter() {defaultValue=-0.2, group=Distortion_Macro, parameterName=Radial_Panning_Y}
  )
  n38 = TextureSampleParameter2D!(
   UVs=LinearInterpolate(
    A=Panner(Coordinate=@n33, Speed=@n34) {speedY=0.2}
    B=Panner(Coordinate=@n36, Speed=@n37)
    Alpha=ScalarParameter() {group=Distortion_Macro, parameterName=Use_MeshUV_OR_ScreenBasedRadial, sortPriority=0}
   )
   Apply View MipBias
  ) {
   group=Distortion_Macro
   parameterName=Macro_Distortion_Texture
   samplerSource=SSM_Wrap_WorldGroupSettings
   texture=&AtmosphericCloudNoise01
  }
  n39 = ScalarParameter() {defaultValue=1, group=Distortion_Macro, parameterName=Macro_Distortion_Mult}
  n40 = Clamp(None=Power(Base=Multiply(A=@n38.R, B=@n39), Exp=ScalarParameter() {defaultValue=1, group=Distortion_Macro, parameterName=Macro_Distortion_Exponent}))
  n41 = Multiply(A=AppendVector(A=@n40, B=Constant())) {constB=2}
  n42 = Multiply(A=AppendVector(A=Constant(), B=@n40)) {constB=2}
  n43 = Add(
   A=LinearInterpolate(
    A=Add(A=@n41) {constB=-1}
    B=Add(A=@n42) {constB=-1}
    Alpha=ScalarParameter() {defaultValue=1, group=Distortion_Macro, parameterName=MultAdd_XorY, sortPriority=7}
   )
   B=AppendVector(
    A=ScalarParameter() {group=Distortion_Macro, parameterName=Macro_Distortion_Offset_X, sortPriority=6}
    B=ScalarParameter() {defaultValue=1, group=Distortion_Macro, parameterName=Macro_Distortion_Offset_Y, sortPriority=7}
   )
  ) {constB=0}
  n44 = ScalarParameter() {defaultValue=0.1, group=Distortion_Macro, parameterName=Macro_Distortion_Intensity, sortPriority=5}
  n45 = ScalarParameter() {
   defaultValue=1
   desc="Large Brush stroke distortion. Good for overall silhouette changes. "
   group=Distortion_Macro
   parameterName=Use_Macro_Distortion
   sortPriority=-1
  }
  n46 = ScalarParameter() {desc=Switches between using UV channel 0 or 1 for mapping distortion, group=1. Base, parameterName=Use_UV1_ForDistortion, sortPriority=-1}
  n47 = MaterialFunctionCall(Float2=LinearInterpolate(A=TextureCoordinate(), B=TextureCoordinate() {coordinateIndex=1}, Alpha=@n46)) {materialFunction=&BreakOutFloat2Components}
  n48 = Multiply(A=@n47.R, B=ScalarParameter() {defaultValue=3, group=Distortion_Micro, parameterName=Distortion_Micro_Tile_X, sortPriority=3})
  n49 = Multiply(A=@n47.G, B=ScalarParameter() {defaultValue=0.5, group=Distortion_Micro, parameterName=Distortion_Micro_Tile_Y, sortPriority=4})
  n50 = ScalarParameter() {defaultValue=0.1, group=Distortion_Micro, parameterName=Distortion_Micro_Panspeed_X, sortPriority=1}
  n51 = ScalarParameter() {defaultValue=0.5, group=Distortion_Micro, parameterName=Distortion_Micro_Panspeed_Y, sortPriority=2}
  n52 = TextureSampleParameter2D(UVs=Panner(Coordinate=AppendVector(A=@n48, B=@n49), Speed=AppendVector(A=@n50, B=@n51)) {speedY=0.2}) {
   group=Distortion_Micro
   parameterName=Micro_Distortion_Texture
   samplerSource=SSM_Wrap_WorldGroupSettings
   texture=/Game/Global/Textures/Default/Noise/T_Dudebro_Noise_01
  }
  n53 = Add(A=Multiply(A=AppendVector(A=@n52.G, B=@n52.G)) {constB=2}) {constB=-1}
  n54 = ScalarParameter() {defaultValue=0.0125, group=Distortion_Micro, parameterName=Micro_Distortion_Intensity, sortPriority=5}
  n55 = ScalarParameter() {
   desc="Small detail distortion. Good for giving extra detail to compliment the silhouette. "
   group=Distortion_Micro
   parameterName=Use_Micro_Distortion
   sortPriority=-1
  }
  n56 = Add(A=LinearInterpolate(B=Multiply(A=@n43, B=@n44) {constB=0.05}, Alpha=@n45), B=LinearInterpolate(B=Multiply(A=@n53, B=@n54), Alpha=@n55))
  n57 = TextureSampleParameter2D(UVs=Add(A=MaterialFunctionCall(R=@n29.R, G=@n30) {materialFunction=&AppendMany}.RG, B=@n56)) {
   desc=Used to control the bottom edge position of the band
   group="2. Edge Distortion Band (Fresnel Distortion = 0)"
   parameterName=EdgeDistortion_Bottom_Texture
   samplerSource=SSM_Clamp_WorldGroupSettings
   texture=/Game/Weapons/FORT_Melee/Pickaxe_Obsidian_Female/Effects/T_FX_Obsidian_FlameDistortion
  }
  n58 = ScalarParameter() {defaultValue=0.5, group=Galaxy, parameterName=GalaxyTiling}
  n59 = MaterialFunctionCall() {materialFunction=/Engine/Functions/Engine_MaterialFunctions02/Utility/FOV}
  n60 = Constant() {r=0.414}
  n61 = Constant() {r=1.091}
  n62 = Constant()
  n63 = Constant() {r=1}
  n64 = LinearInterpolate(
   A=Constant() {r=3}
   B=Constant() {r=1}
   Alpha=Saturate(None=MaterialFunctionCall(Input=@n59."FOV ", Input Low=@n60, Input High=@n61, Target Low=@n62, Target High=@n63) {materialFunction=&RemapValueRange}.Result)
  )
  n65 = Time()
  n66 = LinearInterpolate(A=@n65, B=Multiply(A=@n65, B=Constant()), Alpha=ScalarParameter() {group=Galaxy, parameterName=Background Rotation})
  n67 = AppendVector(
   A=AppendVector(
    A=ScalarParameter() {defaultValue=0.02, group=Galaxy, parameterName=Galaxy Rotation Speed}
    B=ScalarParameter() {defaultValue=-0.02, group=Galaxy, parameterName=Galaxy Rotation Speed_1}
   )
   B=ScalarParameter() {defaultValue=0.1, group=Galaxy, parameterName=Galaxy Rotation Speed_2}
  )
  n68 = TextureObjectParameter() {
   group=Galaxy
   parameterName=GalaxyTexture
   texture=/Game/Characters/Player/Male/Medium/Bodies/M_Med_Soldier_04/Skins/Galaxy/Textures/T_FN_Nebula
  }
  n69 = ScalarParameter() {defaultValue=0.012, group=Galaxy, parameterName=WarpIntensity}
  n70 = Multiply(
   A=VectorParameter() {group=Galaxy, parameterName=GalaxyNebulaColorMult, defaultValue=<rgba 0.3 0.3 0.3 1>}.RGB
   B=MaterialFunctionCall!(
    CameraVector
    Rotation Speed=Multiply(A=@n66, B=@n67)
    Texture=@n68
    Tiling=Multiply(A=@n58, B=@n64)
    Offset=Multiply(B=@n69)
    AxisFadeContrast=Constant() {r=6}
   ) {materialFunction=/Game/Weapons/WeaponSkins/MaterialFunction/TriplanarCameraVector_Rotated}.Result
  )
  n71 = Desaturation(None=@n70, Fraction=ScalarParameter() {defaultValue=0.5, group=Galaxy, parameterName=Star Desaturation})
  n72 = Multiply(A=@n71, B=VectorParameter() {group=Galaxy, parameterName=Small Star Brightness, defaultValue=<rgba 207.7112 21.7887 300 1>}.RGB)
  n73 = Time()
  n74 = ScalarParameter() {defaultValue=-0.0625, group=Galaxy, parameterName=Small Star Rotation speed}
  n75 = TextureObject() {samplerType=SAMPLERTYPE_Masks, texture=/Game/Characters/Player/Male/Medium/Bodies/M_Med_Soldier_04/Skins/Galaxy/Textures/Celestial_T_Stars}
  n76 = ScalarParameter() {defaultValue=5, group=Galaxy, parameterName=Small Star Tiling}
  n77 = ComponentMask(None=MaterialFunctionCall(Rotation Speed=Multiply(A=@n73, B=@n74), Texture=@n75, Tiling=Multiply(A=@n76, B=@n64)) {materialFunction=&TriplanarCameraVector_Rotated}.Result) {b=true}
  n78 = Multiply(A=@n77, B=@n77)
  n79 = Multiply(A=@n70, B=ScalarParameter() {defaultValue=1, group=Galaxy, parameterName=Galaxy_Brightness})
  n80 = VectorParameter() {group=Galaxy, parameterName=Large Star Brightness, defaultValue=<rgba 657.7523 68.9975 950 1>}
  n81 = TextureObject() {samplerType=SAMPLERTYPE_Masks, texture=&Celestial_T_Stars}
  n82 = ScalarParameter() {defaultValue=2, group=Galaxy, parameterName=Bright Star Tiling}
  n83 = ComponentMask(None=MaterialFunctionCall!(CameraVector, Texture=@n81, Tiling=Multiply(A=@n82, B=@n64), Offset, AxisFadeContrast) {materialFunction=/Engine/Functions/Engine_MaterialFunctions01/Texturing/TriplanarCameraVector}.Result) {r=true}
  n84 = Multiply(A=@n83, B=@n83)
  n85 = TextureSample(UVs=Panner(Coordinate=TextureCoordinate() {uTiling=2, vTiling=2}) {speedX=0.133, speedY=0.2}) {
   samplerSource=SSM_Wrap_WorldGroupSettings
   samplerType=SAMPLERTYPE_Masks
   texture=&Celestial_T_Stars
  }
  n86 = Multiply(
   A=ScalarParameter() {defaultValue=1, group=Galaxy, parameterName=Galaxy_GlobalBrightness}
   B=LinearInterpolate(A=LinearInterpolate(A=@n79, B=@n72, Alpha=@n78), B=Multiply(A=@n71, B=@n80.RGB), Alpha=Multiply(A=@n84, B=@n85.G))
  )
  n87 = ScalarParameter() {
   desc=Uses the object's fresnel as the UV space for the edge distortion instead of the normal UV mapping.
   group=1. Base
   parameterName=Use_FresnelDistortion
   sortPriority=0
  }
  n88 = MaterialFunctionCall(Float2=Multiply(
    A=TextureCoordinate()
    B=ScalarParameter() {defaultValue=1, group="2. Edge Distortion Band (Fresnel Distortion = 0)", parameterName=EdgeDistortion_Tiling}
   )) {materialFunction=&BreakOutFloat2Components}
  n89 = TextureCoordinate()
  n90 = AppendVector(
   A=ScalarParameter() {defaultValue=1, group="2. Edge Distortion Band (Fresnel Distortion = 0)", parameterName=TopEdge_LerpMask_Tiling_X}
   B=ScalarParameter() {group="2. Edge Distortion Band (Fresnel Distortion = 0)", parameterName=TopEdge_LerpMask_Tiling_Y}
  )
  n91 = ScalarParameter() {defaultValue=-0.25, group="2. Edge Distortion Band (Fresnel Distortion = 0)", parameterName=TopEdge_LerpMask_Speed_X}
  n92 = ScalarParameter() {defaultValue=0.1, group="2. Edge Distortion Band (Fresnel Distortion = 0)", parameterName=TopEdge_LerpMask_Speed_Y}
  n93 = TextureSampleParameter2D(UVs=Panner(Coordinate=Multiply(A=@n89, B=@n90), Speed=AppendVector(A=@n91, B=@n92))) {
   desc=Adds extra peaks and valleys by masking the top edge distortion edge height
   group="2. Edge Distortion Band (Fresnel Distortion = 0)"
   parameterName=EdgeDistortion_TopEdge_LerpMask
   texture=/Game/Characters/Player/Male/Medium/Bodies/M_MED_Viper/Textures/Glitch/T_SoftNoise
  }
  n94 = Add(
   A=@n88.G
   B=LinearInterpolate(
    A=ScalarParameter() {defaultValue=0.4, group="2. Edge Distortion Band (Fresnel Distortion = 0)", parameterName=EdgeDistort_OffsetTop_LerpMask_Min}
    B=ScalarParameter() {defaultValue=0.4, group="2. Edge Distortion Band (Fresnel Distortion = 0)", parameterName=EdgeDistort_OffsetTop_LerpMask_Max}
    Alpha=@n93.R
   )
  )
  n95 = ScalarParameter() {desc=Switches between using UV channel 0 or 1 for mapping distortion, group=1. Base, parameterName=Use_UV1_ForDistortion, sortPriority=-1}
  n96 = LinearInterpolate(A=TextureCoordinate(), B=TextureCoordinate() {coordinateIndex=1}, Alpha=@n95)
  n97 = AppendVector(
   A=ScalarParameter() {defaultValue=10, group=Distortion_FlameLicks, parameterName=Lick_Distortion_Tiling_X}
   B=ScalarParameter() {defaultValue=2, group=Distortion_FlameLicks, parameterName=Lick_Distortion_Tiling_Y}
  )
  n98 = ScalarParameter() {defaultValue=0.1, group=Distortion_FlameLicks, parameterName=Lick_Distortion_Speed_X}
  n99 = ScalarParameter() {defaultValue=0.1, group=Distortion_FlameLicks, parameterName=Lick_Distortion_Speed_Y}
  n100 = TextureSampleParameter2D(UVs=Panner(Coordinate=Multiply(A=@n96, B=@n97), Speed=AppendVector(A=@n98, B=@n99))) {
   group=Distortion_FlameLicks
   parameterName=Lick_Distortion_Texture
   samplerType=SAMPLERTYPE_Masks
   texture=/Game/Global/Textures/Default/Noise/T_CellularNoise
  }
  n101 = ScalarParameter() {desc=Switches between using UV channel 0 or 1 for mapping distortion, group=1. Base, parameterName=Use_UV1_ForDistortion, sortPriority=-1}
  n102 = LinearInterpolate(A=TextureCoordinate(), B=TextureCoordinate() {coordinateIndex=1}, Alpha=@n101)
  n103 = AppendVector(
   A=ScalarParameter() {defaultValue=2, group=Distortion_FlameLicks, parameterName=Lick_Mask_Tiling_X, sliderMin=1}
   B=ScalarParameter() {group=Distortion_FlameLicks, parameterName=Lick_Mask_Tiling_Y}
  )
  n104 = ScalarParameter() {defaultValue=-0.1, group=Distortion_FlameLicks, parameterName=Lick_Mask_Speed_X}
  n105 = ScalarParameter() {defaultValue=0.02, group=Distortion_FlameLicks, parameterName=Lick_Mask_Speed_Y}
  n106 = TextureSampleParameter2D(UVs=Panner(Coordinate=Multiply(A=@n102, B=@n103), Speed=AppendVector(A=@n104, B=@n105))) {
   desc=Uses a texture to mask the distortion. Can be used to give gaps with peaks and valleys.
   group=Distortion_FlameLicks
   parameterName=Lick_Mask_Texture
   texture=/Game/Environments/Foliage/Textures/T_NoiseMask03_Sparce
  }
  n107 = Power(Base=@n106.R, Exp=ScalarParameter() {defaultValue=2, group=Distortion_FlameLicks, parameterName=Lick_Mask_Exponent})
  n108 = Multiply(
   A=Clamp(None=OneMinus(None=@n100.R))
   B=LinearInterpolate(
    A=ScalarParameter() {defaultValue=0.01, group=Distortion_FlameLicks, parameterName=Lick_Mask_Intensity_Minimum}
    B=ScalarParameter() {defaultValue=1, group=Distortion_FlameLicks, parameterName=Lick_Mask_Intensity_Maximum}
    Alpha=Clamp(None=@n107)
   )
  )
  n109 = ScalarParameter() {defaultValue=0.1, group=Distortion_FlameLicks, parameterName=Lick_Distortion_Intensity}
  n110 = ScalarParameter() {
   desc=" Adds a distortion that resembles flame-licks. Thin fingers that reach out."
   group=Distortion_FlameLicks
   parameterName=Use_Distortion_Lick
   sortPriority=-1
  }
  n111 = Add(A=@n56, B=LinearInterpolate(B=Multiply(A=@n108, B=@n109) {constB=0.1}, Alpha=@n110))
  n112 = TextureSampleParameter2D(UVs=Add(A=AppendVector(A=@n88.R, B=@n94), B=@n111)) {
   desc=Used to control the top edge position of the band
   group="2. Edge Distortion Band (Fresnel Distortion = 0)"
   parameterName=EdgeDistortion_Top_Texture
   samplerSource=SSM_Clamp_WorldGroupSettings
   texture=&T_FX_Obsidian_FlameDistortion
  }
  n113 = Subtract(A=@n112.RGB, B=@n57.RGB)
  n114 = VectorParameter() {group="2. Edge Distortion Band (Fresnel Distortion = 0)", parameterName=EdgeDistort_Color, defaultValue=<rgba 0.9823 0.4911 0.9004 1>}
  n115 = ScalarParameter() {defaultValue=1, group="3. Edge Distortion Band Fresnel (Fresnel Distortion = 1)", parameterName=Fresnel_Brightness, sortPriority=-1}
  n116 = ScalarParameter() {defaultValue=1, group="3. Edge Distortion Band Fresnel (Fresnel Distortion = 1)", parameterName=Fresnel_Exp, sortPriority=-1}
  n117 = MaterialFunctionCall(Brightness=@n115, Exponent=@n116, Vertex / Pixel Normal=VertexNormalWS()) {materialFunction=&MF_Fresnel_FX}
  n118 = ScalarParameter() {defaultValue=1, group="3. Edge Distortion Band Fresnel (Fresnel Distortion = 1)", parameterName=OuterEdge_Mask_Lerp_0, sortPriority=-1}
  n119 = ScalarParameter() {defaultValue=1.2, group="3. Edge Distortion Band Fresnel (Fresnel Distortion = 1)", parameterName=OuterEdge_Mask_Lerp_1, sortPriority=-1}
  n120 = TextureCoordinate()
  n121 = AppendVector(
   A=ScalarParameter() {defaultValue=1, group="3. Edge Distortion Band Fresnel (Fresnel Distortion = 1)", parameterName=OuterEdge_Mask_Tiling_X}
   B=ScalarParameter() {group="3. Edge Distortion Band Fresnel (Fresnel Distortion = 1)", parameterName=OuterEdge_Mask_Tiling_Y}
  )
  n122 = ScalarParameter() {defaultValue=-0.25, group="3. Edge Distortion Band Fresnel (Fresnel Distortion = 1)", parameterName=OuterEdge_Mask_Speed_X}
  n123 = ScalarParameter() {defaultValue=0.1, group="3. Edge Distortion Band Fresnel (Fresnel Distortion = 1)", parameterName=OuterEdge_Mask_Speed_Y}
  n124 = TextureSampleParameter2D(UVs=Panner(Coordinate=Multiply(A=@n120, B=@n121), Speed=AppendVector(A=@n122, B=@n123))) {
   desc=Adds extra peaks and valleys by masking the top edge distortion edge height
   group="3. Edge Distortion Band Fresnel (Fresnel Distortion = 1)"
   parameterName=OuterEdge_Mask_LerpTexture
   texture=&T_SoftNoise
  }
  n125 = DynamicParameter() {
   desc=Can be used to control the vertical position of the top and bottom band thickness.
   paramNames [Outer Edge Exponent, Inner Edge Exponent, Param3, Param4]
  }
  n126 = Multiply(A=LinearInterpolate(A=@n118, B=@n119, Alpha=@n124.R), B=@n125.Outer Edge Exponent)
  n127 = TextureSampleParameter2D(UVs=Add(A=Power(Base=@n117.Fresnel Double Sided, Exp=@n126), B=@n111)) {
   desc=Used to control the Outer edge position of the band
   group="3. Edge Distortion Band Fresnel (Fresnel Distortion = 1)"
   parameterName=Fresnel_EdgeDistortion_Outside_Texture
   samplerSource=SSM_Clamp_WorldGroupSettings
   texture=&T_FX_Obsidian_FlameDistortion
  }
  n128 = ScalarParameter() {
   defaultValue=1.5
   group="3. Edge Distortion Band Fresnel (Fresnel Distortion = 1)"
   parameterName=FresnelDistortEdge_InsideExp
   sortPriority=-1
  }
  n129 = Multiply(A=@n125.Inner Edge Exponent, B=@n128)
  n130 = TextureSampleParameter2D(UVs=Add(A=@n56, B=Power(Base=@n117.Fresnel Double Sided, Exp=@n129))) {
   desc=Used to control the inside edge position of the band
   group="3. Edge Distortion Band Fresnel (Fresnel Distortion = 1)"
   parameterName=Fresnel_EdgeDistortion_Inside_Texture
   samplerSource=SSM_Clamp_WorldGroupSettings
   texture=&T_FX_Obsidian_FlameDistortion
  }
  n131 = Subtract(A=@n127.RGB, B=@n130.RGB)
  n132 = VectorParameter() {
   group="3. Edge Distortion Band Fresnel (Fresnel Distortion = 1)"
   parameterName=FresnelDistortEdge_Color
   defaultValue=<rgba 0.9823 0.4911 0.9004 1>
  }
  n133 = Subtract(A=CameraPositionWS(), B=ObjectPositionWS())
  n134 = VectorParameter() {group=LookatOpacity, parameterName=ObjectVectorToCompare, defaultValue=<rgba 1 0 0 1>}
  n135 = Abs(None=DotProduct(A=Normalize(VectorInput=@n133), B=Transform(None=@n134.RGB) {transformSourceType=TRANSFORMSOURCE_Local}))
  n136 = ScalarParameter() {
   defaultValue=1
   desc=At what point in the dot product do you want to start the transition
   group=LookatOpacity
   parameterName="TransitionStart-Lookat (0-1)"
  }
  n137 = ScalarParameter() {desc=At what point in the dot product do you want to end the transition, group=LookatOpacity, parameterName="TransitionEnd-Lookat (0-1)"}
  n138 = Constant()
  n139 = Constant() {r=1}
  n140 = LinearInterpolate(Alpha=LinearInterpolate(
    A=ScalarParameter() {group=LookatOpacity, parameterName=Brightness-LookAway}
    B=ScalarParameter() {defaultValue=1, group=LookatOpacity, parameterName=Brightness-LookAt}
    Alpha=MaterialFunctionCall(Input=@n135, Input Low=@n136, Input High=@n137, Target Low=@n138, Target High=@n139) {materialFunction=&RemapValueRange}.Result
   )) {constA=0.1}
  n141 = ScalarParameter() {defaultValue=10, group=Edge Band Emissive Noise, parameterName=EdgeNoise_Emissive_Max}
  n142 = Multiply(
   A=TextureCoordinate()
   B=AppendVector(
    A=ScalarParameter() {defaultValue=15, group=Edge Band Emissive Noise, parameterName=EdgeNoise_Tile_X}
    B=ScalarParameter() {defaultValue=1.5, group=Edge Band Emissive Noise, parameterName=EdgeNoise_Tile_Y}
   )
  )
  n143 = ScalarParameter() {defaultValue=0.2, group=Edge Band Emissive Noise, parameterName=EdgeNoise_A_Panner_X}
  n144 = ScalarParameter() {defaultValue=0.5, group=Edge Band Emissive Noise, parameterName=EdgeNoise_A_Panner_Y}
  n145 = TextureSampleParameter2D(UVs=Panner(Coordinate=@n142, Speed=AppendVector(A=@n143, B=@n144))) {
   desc=Adds noise to the emissive channel on the edge band
   group=Edge Band Emissive Noise
   parameterName=EdgeNoise_Texture
   texture=&T_SoftNoise
  }
  n146 = ScalarParameter() {defaultValue=-0.2, group=Edge Band Emissive Noise, parameterName=EdgeNoise_B_Panner_X}
  n147 = ScalarParameter() {defaultValue=0.2, group=Edge Band Emissive Noise, parameterName=EdgeNoise_B_Panner_Y}
  n148 = TextureSampleParameter2D(UVs=Panner(Coordinate=@n142, Speed=AppendVector(A=@n146, B=@n147))) {
   desc=Adds noise to the emissive channel on the edge band
   group=Edge Band Emissive Noise
   parameterName=EdgeNoise_Texture
   texture=&T_SoftNoise
  }
  n149 = Subtract(A=@n145.R, B=@n148.R)
  n150 = Multiply(
   A=LinearInterpolate(A=Multiply(A=@n113, B=@n114.RGB), B=Multiply(A=@n131, B=@n132.RGB), Alpha=@n87) {constA=1}
   B=LinearInterpolate(
    A=ScalarParameter() {defaultValue=0.8, group=Edge Band Emissive Noise, parameterName=EdgeNoise_Emissive_Min}
    B=Multiply(A=@n140, B=@n141)
    Alpha=Clamp(None=@n149)
   )
  )
  n151 = Add(A=@n150, B=Multiply(A=@n57.RGB, B=@n86))
  n152 = Add(A=@n150, B=Multiply(A=@n130.RGB, B=@n86))
  n153 = ScalarParameter() {
   desc=Uses the object's fresnel as the UV space for the edge distortion instead of the normal UV mapping.
   group=1. Base
   parameterName=Use_FresnelDistortion
   sortPriority=0
  }
  NamedRerouteDeclaration(None=LinearInterpolate(
    A=@n86
    B=LinearInterpolate(A=@n151, B=@n152, Alpha=@n153)
    Alpha=ScalarParameter() {desc=Adds UV distortion to the silhouette, group=1. Base, parameterName=Use_Distortion, sortPriority=-1}
   )) {name=2D Space Emissive, nodeColor=<rgba 0.6641 0 1 1>}
  n154 = Constant() {r=1}
  n155 = QualitySwitch(Default=DepthFade(FadeDistance=ScalarParameter() {defaultValue=50, parameterName=Depth Fade}), Low=@n154)
  NamedRerouteDeclaration(None=ShadingPathSwitch(Default=FeatureLevelSwitch(Default=@n155, ES2=@n154, ES3_1=@n154, SM4_REMOVED=@n155), Forward=@n154)) {name=Depth Fade, nodeColor=<rgba 1 0.1577 0 1>}
  ObjectPositionWS()
  Power(Base=@n28) {constExponent=1.25}
  Power() {constExponent=0.7}
  n156 = StaticSwitchParameter(
   True=Multiply(A=MaterialFunctionCall() {materialFunction=&/MF_VoidMask_3DTurb}.Result)
   False=Constant()
  ) {defaultValue=true, parameterName=WPO Noise}
  n157 = DynamicParameter() {parameterIndex=3, defaultValue=<rgba 1 0 0 1>, paramNames [Thicken, Shape Shift, Rand, Noise UV Scale]}
  n158 = Multiply(A=ScalarParameter() {defaultValue=1, parameterName=Base}, B=@n157.Thicken)
  n159 = MaterialFunctionCall(WidthBase=@n158, WidthTip=@n158, Expand U or V UV Channel=StaticBool()) {materialFunction=&SplineThicken}
  n160 = Distance(A=WorldPosition().XYZ, B=ParticlePositionWS())
  n161 = ScalarParameter() {defaultValue=0.005, parameterName=Shape Change distance}
  n162 = Saturate(None=OneMinus(None=Multiply(A=@n160, B=@n161)))
  n163 = ScalarParameter() {defaultValue=5, parameterName=Shape Change range Pow}
  n164 = DynamicParameter() {parameterIndex=3, defaultValue=<rgba 1 0 0 1>, paramNames [Thicken, Shape Shift, Rand, Noise UV Scale]}
  n165 = Multiply(A=Multiply(A=VertexNormalWS(), B=Power(Base=@n162, Exp=@n163)), B=@n164.Shape Shift)
  Reroute(None=StaticSwitchParameter(True=Add(A=@n156, B=@n159.WorldPositionOffset), False=Add(A=@n156, B=@n165)) {parameterName=Spline Thicken})
  Saturate(None=Add(A=@n28) {constB=0.5})
  ScalarParameter() {defaultValue=1, parameterName=Tip}
  ScalarParameter() {defaultValue=20}
  n166 = MaterialFunctionCall() {materialFunction=&TangentBasis}
  n167 = DotProduct(
   A=MaterialFunctionCall() {materialFunction=&CameraDirectionVector}.Result
   B=MaterialFunctionCall(X=@n166.X U, Y=@n166.X U, Z=@n166.X U) {materialFunction=&MakeFloat3}.Result
  )
  n168 = Power(Base=OneMinus(None=Abs(None=@n167)), Exp=ScalarParameter() {defaultValue=0.8, parameterName=Fresnel Power}) {constExponent=0.4}
  n169 = NamedRerouteUsage()
  n170 = Constant() {r=0.5}
  n171 = Multiply(
   A=Multiply(A=NamedRerouteUsage(), B=Multiply(A=@n169, B=@n1.A))
   B=SphereMask(A=TextureCoordinate(), B=@n170, Radius=@n170) {hardnessPercent=0.5}
  )
  n172 = StaticSwitchParameter(True=StaticSwitchParameter(True=Multiply(A=@n168, B=@n171), False=@n171) {defaultValue=true, parameterName=Use Fresnel}, False=@n1.A) {parameterName=Spline Thicken}
  n173 = VectorParameter() {parameterName=UVs, defaultValue=<rgba 1 1 1 1>}
  n174 = Multiply(A=Time(), B=ScalarParameter() {defaultValue=0.5, parameterName=Direction Pan Speed})
  n175 = Constant2Vector() {g=-0.3, r=1}
  n176 = DynamicParameter() {parameterIndex=3, defaultValue=<rgba 1 0 0 1>, paramNames [Thicken, Shape Shift, Rand, Noise UV Scale]}
  n177 = Constant2Vector() {r=1}
  n178 = Add(
   A=Multiply(A=TextureCoordinate(), B=AppendVector(A=@n173.B, B=@n173.B))
   B=Add(A=Multiply(A=@n174, B=@n175), B=Multiply(A=@n176.Rand, B=@n177))
  )
  n179 = StaticSwitchParameter(
   True=Multiply(
    A=@n172
    B=TextureSample(UVs=@n178) {samplerType=SAMPLERTYPE_Grayscale, texture=/Game/Effects/Fort_Effects/Textures/Fire/T_WispyFireNoise_Soft_M}.RGB
   )
   False=@n172
  ) {parameterName=IS_Direction FX}
  n180 = Multiply(
   A=LinearInterpolate(A=@n179, B=Saturate(None=@n179), Alpha=ScalarParameter() {parameterName=SaturateAlphaBeforeDepthFade})
   B=ScalarParameter() {defaultValue=1, parameterName=Alpha}
  )
  n181 = StaticSwitchParameter(True=Multiply(A=@n180, B=NamedRerouteUsage()), False=@n180) {defaultValue=true, parameterName=Use_Depth_Fade}
  n182 = ScalarParameter() {defaultValue=100, parameterName=CameraFade_Distance}
  StaticSwitchParameter(
   True=Multiply(A=@n181, B=MaterialFunctionCall(Distance=@n182) {materialFunction=&Near_Camera_Fade}.Result)
   False=@n181
  ) {defaultValue=true, parameterName=Use_CameraFade}
  TextureCoordinate()
 ]
}
