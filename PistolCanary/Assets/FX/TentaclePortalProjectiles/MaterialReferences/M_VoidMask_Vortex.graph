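# Translucent unlit Niagara sprite and mesh particle material that draws a swirling, eroding vortex mask, apparently for tentacle portal projectile FX.
# Panning wispy noise in twistable polar UVs, modulated by a second noise and a sphere mask, forms the shape.
# Dynamic parameters drive twist, smoothstep erosion and a noise-warped sine mask swipe.
# The result blends dark blue to white times particle color, while alpha adds a UV edge fade and quality-switched depth fade.
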Material /PistolCanary/Assets/FX/TentaclePortalProjectiles/MaterialReferences/M_VoidMask_Vortex
graph {
 properties {
  bAutomaticallySetUsageInEditor=false
  bUsedWithNiagaraMeshParticles=true
  bUsedWithNiagaraSprites=true
  blendMode=BLEND_Translucent
  shadingModel=MSM_Unlit
  usedShadingModels=Unlit
 }
 nodes [
  Constant2Vector() {g=0.3, r=1}
  Constant2Vector() {g=0.7, r=1}
  ConstantBiasScale(None=Sine(None=Time()))
  Constant() {r=-1}
  Constant() {r=-1}
  Constant() {r=-1}
  n1 = Constant3Vector() {constant=<rgba 0.013 0.013 0.1 0>}
  n2 = Constant3Vector() {constant=<rgba 1 1 1 0>}
  n3 = ParticleColor()
  n4 = DynamicParameter() {defaultValue=<rgba 0 5 0 0>, paramNames [MaskSwipe, MaskPower, Erode, TwistIntensity]}
  n5 = AppendVector(
   A=ScalarParameter() {defaultValue=1, group=Main Texture, parameterName=MainTex_U_Scale, sortPriority=20}
   B=ScalarParameter() {defaultValue=0.3, group=Main Texture, parameterName=MainTex_V_Scale, sortPriority=22}
  )
  n6 = DynamicParameter() {defaultValue=<rgba 0 5 0 0>, paramNames [MaskSwipe, MaskPower, Erode, TwistIntensity]}
  n7 = MaterialFunctionCall!(UV=TextureCoordinate(), Twist Center=@n6.TwistIntensity, Twist Edges) {materialFunction=/Game/Accessories/FORT_Backpacks/Backpack_F_MED_Celestial_Glow/FX/Materials/MF_PolarUV}
  n8 = ScalarParameter() {defaultValue=0.2, group=Main Texture, parameterName=MainTex_U_Speed, sortPriority=25}
  n9 = ScalarParameter() {defaultValue=0.7, group=Main Texture, parameterName=MainTex_V_Speed, sortPriority=27}
  n10 = TextureSampleParameter2D!(
   UVs=Panner(Coordinate=Multiply(A=@n5, B=@n7.Polar UVs), Speed=AppendVector(A=@n8, B=@n9)) {speedX=0.2, speedY=0.7}
   "DDX(UVs)"=@n7.DDX
   "DDY(UVs)"=@n7.DDX
   Apply View MipBias
  ) {
   group=Main Texture
   mipValueMode=TMVM_Derivative
   parameterName=TEx1
   samplerType=SAMPLERTYPE_Grayscale
   texture=/Game/Effects/Fort_Effects/Textures/Fire/T_WispyFireNoise_Soft_M
  }
  n11 = TextureCoordinate()
  n12 = TextureSampleParameter2D!(UVs=Panner(Coordinate=Multiply(A=@n11)) {speedX=0.1, speedY=-0.2}, Apply View MipBias) {
   parameterName=TexNoise
   samplerType=SAMPLERTYPE_Normal
   texture=/Game/Characters/Player/Male/Large/Bodies/M_LRG_Clash_V/Textures/Simple_Perlin_N
  }
  n13 = LinearInterpolate(Alpha=@n12.R) {constA=0.7}
  n14 = TextureCoordinate()
  n15 = Constant() {r=0.5}
  n16 = Saturate(None=Multiply(A=Multiply(A=@n10.R, B=@n13), B=SphereMask(A=@n14, B=@n15, Radius=@n15) {hardnessPercent=1}))
  n17 = MaterialFunctionCall(UV=TextureCoordinate()) {materialFunction=&MF_PolarUV}
  n18 = AppendVector(
   A=ScalarParameter() {defaultValue=1, group=Warp, parameterName=WarpTex_U_Scale, sortPriority=20}
   B=ScalarParameter() {defaultValue=0.7, group=Warp, parameterName=WarpTex_V_Scale, sortPriority=22}
  )
  n19 = ScalarParameter() {defaultValue=1, group=Warp, parameterName=WarpTex_U_Speed, sortPriority=25}
  n20 = ScalarParameter() {defaultValue=-0.6, group=Warp, parameterName=WarpTex_V_Speed, sortPriority=27}
  n21 = TextureSampleParameter2D!(
   UVs=Panner(Coordinate=Multiply(A=@n17.Polar UVs, B=@n18), Speed=AppendVector(A=@n19, B=@n20)) {speedX=1, speedY=-0.6}
   "DDX(UVs)"=@n17.DDX
   "DDY(UVs)"=@n17.DDX
   Apply View MipBias
  ) {
   group=Warp
   mipValueMode=TMVM_Derivative
   parameterName=Tex2
   samplerType=SAMPLERTYPE_Masks
   texture=/Game/Characters/Player/Female/Medium/Bodies/F_MED_Neon_Tight_Suit/Materials/Soft_Perlin_output
  }
  n22 = Multiply(A=@n17.Polar UVs, B=Multiply(A=@n21.R, B=ScalarParameter() {defaultValue=1, group=Warp, parameterName=Warp_intensity}))
  n23 = Multiply(A=MaterialFunctionCall(Float2=Add(A=@n17.Polar UVs, B=@n22)) {materialFunction=&BreakOutFloat2Components}.G)
  n24 = Add(A=@n4.MaskSwipe, B=@n23)
  n25 = Sine(None=Multiply(A=Clamp(None=@n24)) {constB=0.5})
  n26 = Power(Base=@n25, Exp=@n4.MaskPower) {constExponent=4}
  n27 = Multiply(A=@n3.A, B=Multiply(A=SmoothStep(Min=@n4.Erode, Value=@n16) {constMin=-1}, B=Saturate(None=@n26)))
  n28 = Power(Base=Saturate(None=@n27), Exp=ScalarParameter() {defaultValue=1, parameterName=ColorGradientPower}) {constExponent=1}
  MaterialFunctionCall(Emissive=Multiply(A=LinearInterpolate(A=@n1, B=@n2, Alpha=@n28), B=@n3.RGB)) {materialFunction=&MF_FX_EmissiveAdjustments}
  Min(A=Power(Base=@n25), B=OneMinus(None=@n23))
  n29 = MaterialFunctionCall(Float2=TextureCoordinate()) {materialFunction=&BreakOutFloat2Components}
  n30 = Multiply(A=@n29.R) {constB=0.5}
  n31 = Multiply(A=@n29.G) {constB=0.5}
  n32 = Multiply(A=Multiply(A=Sine(None=@n30), B=Sine(None=@n31))) {constB=6}
  Multiply(A=Multiply(A=@n27, B=Saturate(None=@n32)), B=NamedRerouteUsage())
  Multiply(A=@n6.TwistIntensity) {constA=0.3, constB=-1}
  n33 = Constant() {r=1}
  n34 = QualitySwitch(Default=DepthFade(FadeDistance=ScalarParameter() {defaultValue=20, parameterName=DepthFade_Distance}), Low=@n33)
  NamedRerouteDeclaration(None=ShadingPathSwitch(Default=FeatureLevelSwitch(Default=@n34, ES2=@n33, ES3_1=@n33, SM4_REMOVED=@n34), Forward=@n33)) {name=Depth Fade, nodeColor=<rgba 1 0.1577 0 1>}
  ScalarParameter() {parameterName=Smoothstep}
 ]
}
